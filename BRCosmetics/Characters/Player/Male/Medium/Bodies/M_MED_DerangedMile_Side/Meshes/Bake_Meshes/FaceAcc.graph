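# FaceAcc material in the Bake_Meshes folder of the M_MED_DerangedMile_Side male player cosmetic, apparently for a face accessory.
# It is a Default Lit material whose graph holds only a constant dark olive colour, with no textures, parameters or visible output links.
# Its Interchange import data suggests it was created by a mesh import as a simple stand-in rather than an authored shader.
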Material /BRCosmetics/Characters/Player/Male/Medium/Bodies/M_MED_DerangedMile_Side/Meshes/Bake_Meshes/FaceAcc
graph {
 properties {assetImportData="<self>:InterchangeAssetImportData_0", usedShadingModels=Default Lit}
 nodes [Constant3Vector() {constant=<rgba 0.0857 0.087 0.0232 1>}]
}
